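# Terrain material function that adds mask-driven color variation to an input base texture.
# A terrain mask texture, sampled by default on landscape coordinates, drives three chained blends: blue adds a highlight tint, low red darkens, green applies a darker tint.
# World vertex normal Z through CheapContrast blends back to the untouched base; the negative default contrast inverts it, confining variation to flatter, upward-facing ground.
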MaterialFunction /FigmentCore/Athena/Prototype/Terrain/MF_TerrainColorVariation
graph {nodes [
  Constant3Vector() {constant=<rgba 1 1 1 0>}
  n1 = FunctionInput() {inputName=BaseTexture}
  n2 = TextureSample(
   UVs=FunctionInput(Preview=LandscapeLayerCoords() {mappingScale=1525}) {bUsePreviewValueAsDefault=true, inputName=UV Coordinates, inputType=FunctionInput_Vector2}
   Tex=FunctionInput(Preview=TextureObject() {samplerType=SAMPLERTYPE_LinearColor, texture=/Game/Athena/Prototype/Terrain/Textures/T_Athena_Terrain_Mask}) {bUsePreviewValueAsDefault=true, inputName=Mask 2, inputType=FunctionInput_Texture2D}
  ) {samplerType=SAMPLERTYPE_LinearColor, texture=&T_Athena_Terrain_Mask}
  n3 = LinearInterpolate(
   A=@n1
   B=Multiply(A=@n1, B=VectorParameter() {group=Terrain_Color, parameterName=HighLight, defaultValue=<rgba 3 2.337 1.38 0>}.RGB)
   Alpha=@n2.B
  )
  n4 = LinearInterpolate(
   A=Multiply(A=@n3, B=VectorParameter() {group=Terrain_Color, parameterName=EdgeDark, defaultValue=<rgba 0.9 0.9 0.9 0>}.RGB) {constB=0.5154}
   B=@n3
   Alpha=@n2.R
  )
  n5 = Multiply(A=@n4, B=VectorParameter() {defaultValue=<rgba 0.1325 0.205 0.1885 0>}.RGB)
  n6 = ComponentMask(None=VertexNormalWS()) {b=true}
  n7 = ScalarParameter() {defaultValue=-2, parameterName=GrassColorMaskContrast}
  FunctionOutput(None=LinearInterpolate(
    A=LinearInterpolate(A=@n4, B=@n5, Alpha=@n2.G)
    B=@n1
    Alpha=MaterialFunctionCall(In=@n6, Contrast=@n7) {materialFunction=&CheapContrast}.Result
   ))
 ]}
