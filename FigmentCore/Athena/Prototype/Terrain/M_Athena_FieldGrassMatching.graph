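# Masked two-sided foliage material for Athena field grass, with a Nanite override, that color-matches blades to the terrain.
# Color samples the grassland texture in world space, tinted root to tip with terrain variation, and normals blend toward vertex normals.
# The default WPO adds noise-based height and panned-noise wind rotation through Pivot Painter.
# An IsNanite switch selects reroute-based WPO; the graph also builds per-blade wind, physics-field and player-push animation.
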Material /FigmentCore/Athena/Prototype/Terrain/M_Athena_FieldGrassMatching
graph {
 properties {
  bTangentSpaceNormal=false
  bUseMaterialAttributes=true
  bUsedWithInstancedStaticMeshes=true
  bUsedWithNanite=true
  blendMode=BLEND_Masked
  pixelDepthOffsetMode=PDOM_Legacy
  shadingModel=MSM_TwoSidedFoliage
  twoSided=true
  usedShadingModels=Two Sided Foliage
  naniteOverrideMaterial {
   bEnableOverride=true
   overrideMaterial=null
   overrideMaterialEditor=&/M_Athena_FieldGrassMatching_Nanite
   overrideMaterialRef=null
  }
 }
 nodes [
  Clamp(None=MaterialFunctionCall(A=PerInstanceRandom()) {materialFunction=&Lerp_Multiple_Float3}.Lerp 3 Inputs)
  Constant3Vector() {constant=<rgba 0 0 1 0>}
  Constant3Vector()
  Constant3Vector() {constant=<rgba 0 0 1 0>}
  Constant() {r=1}
  Constant()
  n1 = TwoSidedSign()
  n2 = Constant() {r=-1}
  n3 = StaticSwitchParameter(
   True=ComponentMask(None=TextureCoordinate() {coordinateIndex=2}) {g=true}
   False=ComponentMask(None=TextureCoordinate() {coordinateIndex=1}) {g=true}
  ) {parameterName=IsNanite}
  n4 = Multiply(
   A=Power(Base=@n3, Exp=ScalarParameter() {defaultValue=3, parameterName=VertNrmBlndAmt_1})
   B=ScalarParameter() {defaultValue=0.2, parameterName=VertNrmBlndAmt}
  )
  LinearInterpolate(
   A=LinearInterpolate(A=Constant3Vector() {constant=<rgba 0 0 -1 1>}, B=Constant3Vector() {constant=<rgba 0 0 1 1>}, Alpha=@n1)
   B=LinearInterpolate(A=Multiply(A=@n2), Alpha=@n1)
   Alpha=@n4
  )
  n5 = ComponentMask(None=WorldPosition().XYZ) {g=true, r=true}
  n6 = Constant2Vector() {g=-153000, r=-153000}
  n7 = ComponentMask(None=WorldPosition() {worldPositionShaderOffset=WPT_ExcludeAllShaderOffsets}.XYZ) {g=true, r=true}
  n8 = ScalarParameter() {defaultValue=800, group=ModifiedGrass, parameterName=GrassColorMaskSize_2}
  n9 = TextureSampleParameter2D!(UVs=Multiply(A=Divide(A=@n7, B=@n8), B=Constant() {r=-1}), Bias, Apply View MipBias) {
   constMipValue=5
   mipValueMode=TMVM_MipBias
   parameterName=GrassTextureToMatch
   texture=/FigmentCore/Packages/DS_Fortnite_Terrain_NoLOD/Textures/T_Grasslands_Athena_D
  }
  n10 = Multiply(A=@n9.RGB, B=VectorParameter() {parameterName=GrassSample_Tint, defaultValue=<rgba 1 1 1 0>}.RGB)
  n11 = Multiply(A=VectorParameter() {parameterName=GrassTipTint, defaultValue=<rgba 1.2 1.2 1.2 1>}.RGB, B=@n10)
  n12 = Multiply(A=@n11, B=ScalarParameter() {defaultValue=1, parameterName=Brightness_Top})
  n13 = Multiply(A=@n10, B=ScalarParameter() {defaultValue=0.9, parameterName=Brightness_Bottom})
  n14 = LinearInterpolate(A=ScalarParameter() {parameterName=254rss}, B=ScalarParameter() {defaultValue=1.5, parameterName=cvew}, Alpha=@n3) {constA=1.2, constB=-1}
  n15 = Clamp(None=LinearInterpolate(
    A=StaticSwitchParameter(True=@n12, False=@n11) {parameterName=IsNanite}
    B=StaticSwitchParameter(True=@n13, False=@n10) {parameterName=IsNanite}
    Alpha=Clamp(None=@n14)
   )) {clampMode=CMODE_ClampMax}
  n16 = MaterialFunctionCall!(Mask 2, UV Coordinates=Divide(A=Subtract(A=@n5, B=@n6), B=Constant() {r=306200}), BaseTexture=@n15) {materialFunction=&/MF_TerrainColorVariation}
  n17 = Power(Base=TextureSample() {samplerType=SAMPLERTYPE_LinearColor, texture=/Game/Athena/Prototype/Terrain/T_GrassGradient_01}.RGB) {constExponent=1.863}
  n18 = VertexNormalWS()
  n19 = Power(Base=@n3, Exp=ScalarParameter() {defaultValue=1, parameterName=VertNrmBlndAmt_3})
  n20 = ScalarParameter() {defaultValue=1, parameterName=VertNrmBlndAmt_2}
  n21 = LinearInterpolate(
   A=Transform(None=TextureSample() {samplerType=SAMPLERTYPE_Normal, texture=/Game/Environments/Foliage/Textures/T_WavyGrass_01_N}.RGB)
   B=@n18
   Alpha=Clamp(None=Multiply(A=@n19, B=@n20))
  )
  n22 = NamedRerouteUsage()
  n23 = NamedRerouteUsage()
  n24 = ScalarParameter() {group=WPO, parameterName=StaticWPOLow}
  n25 = ScalarParameter() {defaultValue=30, group=WPO, parameterName=StaticWPOHigh}
  n26 = ComponentMask(None=WorldPosition().XYZ) {g=true, r=true}
  n27 = TextureSample(UVs=Divide(A=@n26, B=ScalarParameter() {defaultValue=1000, group=WPO, parameterName=HeightVarianceScale})) {samplerType=SAMPLERTYPE_LinearColor, texture=/Game/Environments/Foliage/Textures/T_FieldGrass_01_M}
  n28 = Multiply(A=Multiply(A=Constant3Vector() {constant=<rgba 0 0 1 0>}, B=LinearInterpolate(A=@n24, B=@n25, Alpha=@n27.R)), B=OneMinus(None=@n3))
  n29 = Panner(
   Coordinate=Divide(A=@n26, B=ScalarParameter() {defaultValue=1200, group=WPO, parameterName=WindWaveScale})
   Time=Multiply(A=Time(), B=ScalarParameter() {defaultValue=0.04, group=WPO, parameterName=WindSpeed})
  ) {speedY=1}
  n30 = ComponentMask(None=Multiply(
    A=TextureSample(UVs=@n29) {samplerType=SAMPLERTYPE_Normal, texture=/Game/Global/Textures/Default/Noise/T_SwirlNoise_N}.RGB
    B=ScalarParameter() {defaultValue=0.5, group=WPO, parameterName=WindPower}
   )) {g=true, r=true}
  n31 = Constant() {r=1}
  n32 = MaterialFunctionCall() {materialFunction=/Engine/Functions/Engine_MaterialFunctions02/PivotPainter/PivotPainter_PerObjectData}
  n33 = DotProduct(A=@n32.X-Axis Vector, B=Subtract(A=WorldPosition().XYZ, B=@n32.Pivot Position))
  n34 = RotateAboutAxis(
   NormalizedRotationAxis=Normalize(VectorInput=AppendVector(A=@n30, B=@n31))
   RotationAngle=Multiply(A=@n27.R, B=ScalarParameter() {defaultValue=0.4, parameterName=Ambient Wind Motion Amount})
   PivotPoint=Subtract(A=WorldPosition().XYZ, B=Multiply(A=@n32.X-Axis Vector, B=@n33))
   Position=Add(A=@n28, B=WorldPosition() {worldPositionShaderOffset=WPT_ExcludeAllShaderOffsets}.XYZ)
  )
  MakeMaterialAttributes!(
   BaseColor=StaticSwitchParameter(True=LinearInterpolate(A=@n16.Result, B=@n15, Alpha=@n17), False=@n15) {defaultValue=true, parameterName=UseTerrainColor}
   Metallic
   Specular=ScalarParameter() {defaultValue=0.05, parameterName=Spec}
   Roughness=ScalarParameter() {defaultValue=1, parameterName=Rough}
   Anisotropy
   EmissiveColor
   Opacity=ScalarParameter() {defaultValue=0.5, parameterName=SSSOpacity}
   OpacityMask=Multiply(A=TextureSample() {texture=/Game/Environments/Foliage/Textures/T_WavyGrass_01}.G, B=PerInstanceFadeAmount())
   Normal=QualitySwitch(Default=@n21, Low=@n18, High=@n21, Medium=@n21)
   Tangent
   WorldPositionOffset=StaticSwitchParameter(True=Add(A=@n22, B=@n23), False=Add(A=@n28, B=@n34)) {parameterName=IsNanite}
   SubsurfaceColor=Constant3Vector()
   ClearCoat
   ClearCoatRoughness
   AmbientOcclusion
   Refraction
   CustomizedUVs_0
   CustomizedUVs_1
   CustomizedUVs_2
   CustomizedUVs_3
   CustomizedUVs_4
   CustomizedUVs_5
   CustomizedUVs_6
   CustomizedUVs_7
   PixelDepthOffset
   ShadingModel
   Displacement
  )
  n35 = TextureCoordinate() {coordinateIndex=1}
  n36 = MaterialFunctionCall(X=ComponentMask(None=@n35) {r=true}, Y=ComponentMask(None=@n35) {g=true}, Z=Constant()) {materialFunction=&MakeFloat3}
  n37 = TransformPosition(None=@n36.Result) {transformSourceType=TRANSFORMPOSSOURCE_Instance, transformType=TRANSFORMPOSSOURCE_World}
  NamedRerouteDeclaration(None=@n37) {name=PivotPosition, nodeColor=<rgba 0 0.9766 1 1>}
  n38 = VertexColor() {desc=Meshes using this material need the SimpleGrassPivotPainter.hda script run on them}
  NamedRerouteDeclaration(None=@n38.B) {name=B_VerticalGradient, desc="Vertical gradient - bottom=0 top=1", nodeColor=<rgba 1 0 0.0703 1>}
  n39 = SamplePhysicsVectorField!(World Position=NamedRerouteUsage()) {fieldTarget=Vector_LinearVelocity}
  n40 = MaterialFunctionCall(Vector 3=Add(A=@n39) {constB=0.0001}) {materialFunction=/Engine/Functions/Engine_MaterialFunctions02/Utility/VectorLength}
  n41 = Constant()
  NamedRerouteDeclaration(None=FeatureLevelSwitch(Default=Saturate(None=@n40.V3 Length), ES3_1=@n41)) {name=PhysicsFieldMask, nodeColor=<rgba 0 1 0.3906 1>}
  n42 = Constant3Vector() {constant=<rgba 0 0 1 0>}
  NamedRerouteDeclaration(None=FeatureLevelSwitch(
    Default=MaterialFunctionCall(Vector=@n39, Default=@n42) {materialFunction=&SafeNormalize}.Result
    ES3_1=Constant3Vector() {constant=<rgba 0 0 1 0>}
   )) {name=PhysicsFieldDirection, nodeColor=<rgba 0.3828 0 1 1>}
  n43 = Min(A=@n40.V3 Length, B=ScalarParameter() {defaultValue=1, group=GrassTypeSettings_WPO, parameterName=MaxRotation_PhysicsField_Clamp})
  n44 = Time()
  n45 = Multiply(A=ScalarParameter() {defaultValue=351, group=GrassTypeSettings_WPO, parameterName=PhysicsField_NoiseOffset}, B=NamedRerouteUsage())
  n46 = Frac(None=Multiply(
    A=ScalarParameter() {defaultValue=2.5, group=GrassTypeSettings_WPO, parameterName=MaxRotation_PhysicsField_NoiseSpeed}
    B=Add(A=@n44, B=@n45)
   ))
  n47 = OneMinus(None=@n46)
  n48 = LinearInterpolate(
   A=Multiply(A=Multiply(A=@n46, B=@n47)) {constB=2}
   Alpha=ScalarParameter() {defaultValue=0.6, group=GrassTypeSettings_WPO, parameterName=MaxRotation_PhysicsField_MInBend}
  )
  NamedRerouteDeclaration(None=FeatureLevelSwitch(Default=Multiply(A=@n43, B=@n48), ES3_1=@n41)) {name=PhysicsFieldStrength, nodeColor=<rgba 1 0 0.2109 1>}
  n49 = NamedRerouteUsage()
  n50 = NamedRerouteUsage()
  n51 = NamedRerouteUsage()
  NamedRerouteDeclaration(None=LinearInterpolate(A=LinearInterpolate(A=@n49, B=@n50, Alpha=@n51), B=NamedRerouteUsage(), Alpha=NamedRerouteUsage())) {name=BlendedDirectionVector, nodeColor=<rgba 0 1 0.3203 1>}
  n52 = Multiply(A=NamedRerouteUsage(), B=ScalarParameter() {defaultValue=0.38, group=GrassTypeSettings_WPO, parameterName=MaxRotation_texAnim})
  n53 = Multiply(A=NamedRerouteUsage(), B=ScalarParameter() {defaultValue=1, group=GrassTypeSettings_WPO, parameterName=MaxRotation_PhysicsField})
  n54 = NamedRerouteUsage()
  NamedRerouteDeclaration(None=LinearInterpolate(
    A=LinearInterpolate(A=@n52, B=@n53, Alpha=@n54)
    B=ScalarParameter() {defaultValue=0.375, group=GrassTypeSettings_WPO, parameterName=MaxRotation_PlayerPush}
    Alpha=NamedRerouteUsage()
   )) {name=AnimStrength, nodeColor=<rgba 1 0.3047 0 1>}
  n55 = Normalize(VectorInput=CrossProduct(A=Constant3Vector() {constant=<rgba 0 0 1 0>}, B=NamedRerouteUsage()))
  n56 = Multiply(
   A=NamedRerouteUsage()
   B=LinearInterpolate(A=ScalarParameter() {defaultValue=0.5, group=GrassTypeSettings_WPO, parameterName=AnimBendFalloff}, Alpha=NamedRerouteUsage())
  )
  n57 = StaticSwitchParameter(True=NamedRerouteUsage(), False=NamedRerouteUsage()) {parameterName=UseCameraRelativeOptimization}
  n58 = StaticSwitchParameter(True=WorldPosition() {worldPositionShaderOffset=WPT_CameraRelativeNoOffsets}.XYZ, False=WorldPosition().XYZ) {parameterName=UseCameraRelativeOptimization}
  n59 = NamedRerouteUsage()
  n60 = LinearInterpolate(A=ScalarParameter() {defaultValue=0.5, group=GrassTypeSettings_WPO, parameterName=AnimBendFalloff_min}, Alpha=NamedRerouteUsage())
  n61 = ScalarParameter() {defaultValue=5, group=GrassTypeSettings_WPO, parameterName=AnimBendFalloff_Exponent}
  n62 = Multiply(
   A=Multiply(A=NamedRerouteUsage(), B=Power(Base=@n60, Exp=@n61))
   B=ScalarParameter() {defaultValue=150, group=GrassTypeSettings_WPO, parameterName=AnimStrength_Global}
  )
  n63 = StaticSwitchParameter(True=WorldPosition() {worldPositionShaderOffset=WPT_CameraRelativeNoOffsets}.XYZ, False=LocalPosition() {includedOffsets=ExcludeOffsets}.XYZ) {parameterName=UseCameraRelativeOptimization}
  n64 = StaticSwitchParameter(True=NamedRerouteUsage(), False=NamedRerouteUsage()) {parameterName=UseCameraRelativeOptimization}
  n65 = Subtract(A=Add(A=@n63, B=Multiply(A=@n59, B=@n62)), B=@n64)
  n66 = Length(None=Subtract(A=@n63, B=@n64)) {is=class MaterialExpressionLength}
  n67 = Length(None=@n65)
  n68 = Subtract(A=Multiply(A=@n65, B=Divide(A=@n66, B=@n67)), B=@n63)
  n69 = NamedRerouteUsage()
  n70 = Add(A=Constant() {r=1}, B=ScalarParameter() {defaultValue=1, group=GrassTypeSettings_WPO, parameterName=Stiffness})
  n71 = StaticSwitchParameter(True=@n69, False=Power(Base=@n69, Exp=Clamp(None=@n70) {maxDefault=50, minDefault=1}) {constExponent=3}) {group=GrassTypeSettings, parameterName=IsFlowers}
  n72 = LinearInterpolate(A=ScalarParameter() {defaultValue=-1, group=GrassTypeSettings_WPO, parameterName=LowPolyCounterOffset}, Alpha=@n71)
  n73 = NamedRerouteUsage()
  n74 = NamedRerouteUsage()
  n75 = Multiply(
   A=Multiply(
    A=StaticSwitchParameter(True=@n72, False=@n71) {group=GrassTypeSettings_WPO, parameterName=UseLowPolyCounterOffset}
    B=Multiply(A=@n73, B=@n74)
   )
   B=ScalarParameter() {defaultValue=150, group=GrassTypeSettings_WPO, parameterName=Master_Anim_strength}
  )
  n76 = QualitySwitch(
   Default=StaticSwitchParameter(True=RotateAboutAxis(NormalizedRotationAxis=@n55, RotationAngle=@n56, PivotPoint=@n57, Position=@n58), False=Add(A=@n68, B=@n64)) {group=GrassTypeSettings_WPO, parameterName="UseRealRotationAnimation(MostExpensive)"}
   Low=@n75
   Medium=@n75
  )
  NamedRerouteDeclaration(None=NaniteReplace!(Default=@n75, Nanite=FeatureLevelSwitch(Default=@n76, ES3_1=@n75))) {name=Simple Lookup Animation, nodeColor=<rgba 0 1 0.8359 1>}
  n77 = ScalarParameter() {defaultValue=-50, group=GrassTypeSettings_WPO, parameterName=SinkIntoGroundAmount}
  n78 = Constant()
  n79 = PerInstanceFadeAmount()
  n80 = MaterialFunctionCall(Color=NamedRerouteUsage()) {materialFunction=&SplitComponents}
  n81 = CollectionParameter() {collection=&FortniteMaterialParameters, parameterName=GrassPushDownRadius}
  n82 = Divide(A=PixelDepth().R, B=@n81)
  n83 = Multiply(A=OneMinus(None=Saturate(None=@n82)), B=Constant() {r=-200})
  n84 = Constant()
  n85 = CollectionParameter() {collection=&FortniteMaterialParameters, parameterName=IsPlayerSliding}
  n86 = MaterialFunctionCall(X=@n80.R, Y=@n80.R, Z=Add(A=@n80.B, B=LinearInterpolate(A=@n83, B=@n84, Alpha=@n85))) {materialFunction=&MakeFloat3}
  n87 = Add(A=Multiply(A=LinearInterpolate(A=@n77, B=@n78, Alpha=@n79), B=Constant3Vector() {constant=<rgba 0 0 1 0>}), B=@n86.Result)
  NamedRerouteDeclaration(None=StaticSwitchParameter(
    True=Constant3Vector()
    False=StaticSwitchParameter(True=@n87, False=@n86.Result) {group=GrassTypeSettings, parameterName=IsFallbackMaterial}
   ) {group=GrassTypeSettings_WPO, parameterName=DisableAllAnimation}) {name=WPO-NANITE-out, nodeColor=<rgba 0 1 0.3438 1>}
  NamedRerouteDeclaration(None=@n28) {name=WPOHeightOffset, nodeColor=<rgba 1 0 0.7031 1>}
  NamedRerouteDeclaration(None=Frac(None=@n38.G)) {name=G_RandomPerBlade, desc=Random value per blade, nodeColor=<rgba 0.7109 0 1 1>}
  NamedRerouteDeclaration(None=@n36.Result) {name=PivotPosition_LocalSpace, nodeColor=<rgba 0 0.9766 1 1>}
  NamedRerouteDeclaration(None=TransformPosition(None=@n36.Result) {transformSourceType=TRANSFORMPOSSOURCE_Instance, transformType=TRANSFORMPOSSOURCE_TranslatedWorld}) {name=PivotPosition_CameraRelative, nodeColor=<rgba 0 0.9766 1 1>}
  n88 = StaticSwitchParameter(True=NamedRerouteUsage(), False=WorldPosition().XYZ) {defaultValue=true, group=GrassTypeSettings_WPO, parameterName=UseSamplingPerBladeOrWorldPosition}
  n89 = Multiply(A=NamedRerouteUsage(), B=ScalarParameter() {defaultValue=200, group=GrassTypeSettings_WPO, parameterName=RandomSampleOffsetPerBlade})
  n90 = MaterialFunctionCall!(
   Position=StaticSwitchParameter(True=@n88, False=Add(A=@n89, B=@n88)) {group=GrassTypeSettings_WPO, parameterName="UseRealRotationAnimation(MostExpensive)"}
   IsTrees
   IsGrass=StaticBool() {value=true}
   IsMediumPlant
   IsMediumPlant_2024Anim
   IsMediumPlant2ndLevel
   MediumPlant_BasePivotPostion
  ) {materialFunction=/Game/Environments/Asteria/Foliage/Wind/MF_WindActorData}
  NamedRerouteDeclaration(None=@n90.WindIntensitySingleSample) {name=WindIntensityNoise, nodeColor=<rgba 0.1719 0 1 1>}
  NamedRerouteDeclaration(None=@n90.WindI_Direction_WithNoise) {name=WindDirectionWithNoise, nodeColor=<rgba 0 1 0.9297 1>}
  n91 = CollectionParameter() {collection=&FortniteMaterialParameters, parameterName=PlayerPosition}
  n92 = ComponentMask(None=@n91) {b=true, g=true, r=true}
  n93 = StaticSwitchParameter(
   True=TransformPosition(None=@n92) {transformSourceType=TRANSFORMPOSSOURCE_World, transformType=TRANSFORMPOSSOURCE_TranslatedWorld}
   False=@n92
  ) {parameterName=UseCameraRelativeOptimization}
  n94 = StaticSwitchParameter(True=NamedRerouteUsage(), False=NamedRerouteUsage()) {parameterName=UseCameraRelativeOptimization}
  n95 = Constant() {r=200}
  n96 = Constant()
  NamedRerouteDeclaration(None=Saturate(None=SphereMask(A=@n93, B=@n94, Radius=@n95, Hardness=@n96) {attenuationRadius=64, hardnessPercent=20}) {desc=PlayerSphereMask}) {name=PlayerSphereMask, nodeColor=<rgba 0 0.8828 1 1>}
  NamedRerouteDeclaration(None=Normalize(VectorInput=Subtract(A=@n94, B=@n93))) {name=PlayerPushDirection, nodeColor=<rgba 0 0.6016 1 1>}
  n97 = Constant3Vector() {constant=<rgba 0 0 1 0>}
  n98 = TextureObjectParameter() {
   group=ModifiedGrass
   parameterName=GrassNormal
   samplerSource=SSM_Wrap_WorldGroupSettings
   samplerType=SAMPLERTYPE_Normal
   texture=/Game/VisualThreatManager/StormVisuals/LargeClouds/T_StormCloudDetail01_N
  }
  n99 = MaterialFunctionCall!(
   Diffuse Texture Object
   Spec Texture Object
   TextureSize=ScalarParameter() {defaultValue=1100, group=ModifiedGrass, parameterName=ModifiedGrassSize}
   --------------------------
   Use seperate settings for Z projection
   Use seperate texture for Z projection
   Texture Size Z projection
   Z Plane Diffuse Texture Object
   Z Plane Spec Texture Object
   "--------------------------      "
   ProjectionTransitionContrast
   Use Texture Alpha Heightmap to Refine the Transistion
   HeightLerpTransitionContrast
   PickTheDiffuseHeightChannel
   TheDiffuseHeightChannel
   "--------------------------       "
   NormalTextureObject=@n98
   World Space Normals
   "--------------------------  "
   Export Float 4 Diffuse=StaticBool() {value=true}
   Export Float 4 Spec
   "--------------------------     "
   WorldPosition=WorldPosition() {worldPositionShaderOffset=WPT_ExcludeAllShaderOffsets}.XYZ
   World Space Normal=VertexNormalWS()
  ) {materialFunction=/Engine/Functions/Engine_MaterialFunctions02/Texturing/WorldAlignedTextures_Complex}
  n100 = TextureCoordinate() {coordinateIndex=1}
  n101 = LinearInterpolate(
   A=@n97
   B=@n99.Normal Z Texture
   Alpha=Power(Base=ComponentMask(None=@n100) {g=true}, Exp=ScalarParameter() {defaultValue=2, parameterName=Transition to Upward Normal Mask Power})
  )
  QualitySwitch(Default=@n101, Low=@n97, High=@n101, Medium=@n101)
  n102 = VertexNormalWS()
  n103 = LinearInterpolate(A=@n102, B=@n99.Normal Z Texture, Alpha=Clamp(None=@n4))
  QualitySwitch(Default=@n103, Low=@n102, High=@n103, Medium=@n103)
  Reroute(None=@n28)
  VertexInterpolator(VS=@n37)
  VertexInterpolator(VS=Transform(None=Constant3Vector() {constant=<rgba 0 0 1 0>}) {transformSourceType=TRANSFORMSOURCE_Instance})
 ]
}
