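# Masked, Nanite-enabled landscape material for Athena prototype terrain, blending Grass, GrassNoFoliage, Mud, Forest and Road paint layers.
# Slope-driven height lerps mix grass with two-scale projected rock and gravel, with a static switch for cheaper two-plane projection.
# Landscape grass output spawns four grass types, subtracting road coverage from all and rock from both grass layers.
# Blended results write to a runtime virtual texture, sampled back when virtual texturing is enabled.
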Material /FigmentCore/Athena/Prototype/Terrain/M_Athena_Terrain_Landscape
graph {
 properties {bUseMaterialAttributes=true, bUsedWithNanite=true, blendMode=BLEND_Masked, pixelDepthOffsetMode=PDOM_Legacy, usedShadingModels=Default Lit}
 nodes [
  n1 = AppendVector(
   A=ScalarParameter() {defaultValue=7.3309, group=Rock, parameterName=Rock_Tiling_X}
   B=ScalarParameter() {defaultValue=15.9391, group=Rock, parameterName=Rock_Tiling_Y}
  )
  n2 = Divide(A=LandscapeLayerCoords() {mappingScale=1, mappingType=TCMT_XZ}, B=@n1)
  n3 = TextureSample(UVs=@n2) {samplerSource=SSM_Wrap_WorldGroupSettings, texture=/Game/Packages/DS_Fortnite_Terrain_NoLOD/Textures/T_Cave_Exterior_Wall_D}
  n4 = Divide(A=LandscapeLayerCoords() {mappingScale=1, mappingType=TCMT_YZ}, B=@n1)
  n5 = TextureSample(UVs=@n4) {samplerSource=SSM_Wrap_WorldGroupSettings, texture=&T_Cave_Exterior_Wall_D}
  n6 = VertexNormalWS()
  n7 = MaterialFunctionCall(In=Abs(None=ComponentMask(None=@n6) {r=true}), Contrast=ScalarParameter() {defaultValue=10, parameterName=RockProjContrast_1}) {materialFunction=&CheapContrast}
  n8 = LinearInterpolate(A=@n3.RGB, B=@n5.RGB, Alpha=@n7.Result)
  n9 = AppendVector(
   A=ScalarParameter() {defaultValue=19.716, group=Rock, parameterName=Rock_Tiling_X_1}
   B=ScalarParameter() {defaultValue=43.9378, group=Rock, parameterName=Rock_Tiling_Y_1}
  )
  n10 = Divide(A=LandscapeLayerCoords() {mappingScale=1, mappingType=TCMT_XZ}, B=@n9)
  n11 = TextureSample(UVs=@n10) {samplerSource=SSM_Wrap_WorldGroupSettings, texture=&T_Cave_Exterior_Wall_D}
  n12 = Divide(A=LandscapeLayerCoords() {mappingScale=1, mappingType=TCMT_YZ}, B=@n9)
  n13 = TextureSample(UVs=@n12) {samplerSource=SSM_Wrap_WorldGroupSettings, texture=&T_Cave_Exterior_Wall_D}
  n14 = LinearInterpolate(A=@n11.RGB, B=@n13.RGB, Alpha=@n7.Result)
  n15 = ScalarParameter() {defaultValue=5, parameterName=TransitonRockWall_LOW}
  n16 = ScalarParameter() {defaultValue=-5, parameterName=TransitonRockWall_High}
  n17 = ComponentMask(None=VertexNormalWS()) {b=true}
  n18 = Clamp(None=Add(A=LinearInterpolate(A=@n15, B=@n16, Alpha=@n17), B=ScalarParameter() {defaultValue=0.578, parameterName=TransitonRockWall_Offset}))
  n19 = ScalarParameter() {defaultValue=1534.9661, parameterName=TexSize}
  n20 = VertexNormalWS()
  n21 = ScalarParameter() {defaultValue=0.75, parameterName=RockProjNormalMP_Z}
  n22 = VertexNormalWS()
  n23 = ScalarParameter() {defaultValue=1, parameterName=RockProjNormalMP}
  n24 = ScalarParameter() {defaultValue=3740.7383, parameterName=TexSize_2}
  n25 = VertexNormalWS()
  n26 = ScalarParameter() {defaultValue=0.75, parameterName=RockProjNormalMP_Z_1}
  n27 = VertexNormalWS()
  n28 = ScalarParameter() {defaultValue=1, parameterName=RockProjNormalMP_1}
  n29 = LinearInterpolate(
   A=MaterialFunctionCall!(
    TextureObject=TextureObject() {texture=&T_Cave_Exterior_Wall_D}
    TextureSize=@n19
    WorldPosition
    Export Float 4=StaticBool() {value=true}
    World Space Normal_Z=Multiply(A=@n20, B=@n21)
    TextureObject_Z=TextureObject() {texture=/Game/Packages/DS_Fortnight/WhiteRockCustom_D}
    World Space Normal=Multiply(A=@n22, B=@n23)
    ProjectionTransitionContrast=ScalarParameter() {defaultValue=10, parameterName=RockProjContrast}
   ) {materialFunction=/Game/Athena/Prototype/Terrain/WorldAlignedTexture_ZTex}.XYZ Texture
   B=MaterialFunctionCall(
    TextureObject=TextureObject() {texture=&T_Cave_Exterior_Wall_D}
    TextureSize=@n24
    Export Float 4=StaticBool() {value=true}
    World Space Normal_Z=Multiply(A=@n25, B=@n26)
    TextureObject_Z=TextureObject() {texture=&WhiteRockCustom_D}
    World Space Normal=Multiply(A=@n27, B=@n28)
    ProjectionTransitionContrast=ScalarParameter() {defaultValue=10, parameterName=RockProjContrast_2}
   ) {materialFunction=&WorldAlignedTexture_ZTex}.XYZ Texture
   Alpha=@n18
  )
  n30 = VertexNormalWS()
  n31 = LinearInterpolate(A=LinearInterpolate(A=@n3.A, B=@n5.A, Alpha=@n7.Result), B=LinearInterpolate(A=@n11.A, B=@n13.A, Alpha=@n7.Result), Alpha=@n18)
  n32 = StaticSwitchParameter(True=@n31, False=ComponentMask(None=@n29) {a=true, desc=RockHeight}) {parameterName=UseCheapProjection}
  n33 = ScalarParameter() {defaultValue=-1, group=Rock, parameterName=TransitonRockWall_Offset_TintHeight}
  n34 = Add(
   A=LinearInterpolate(
    A=ScalarParameter() {defaultValue=1.2842, group=Rock, parameterName=TransitonRockWall_LOW_Tint}
    B=ScalarParameter() {defaultValue=-1.4149, group=Rock, parameterName=TransitonRockWall_High_Tint}
    Alpha=ComponentMask(None=@n30) {b=true}
   )
   B=Add(A=Multiply(A=@n32, B=@n33), B=ScalarParameter() {defaultValue=0.5413, group=Rock, parameterName=TransitonRockWall_Offset_Tint})
  )
  n35 = Multiply(
   A=StaticSwitchParameter(True=LinearInterpolate(A=@n8, B=@n14, Alpha=@n18), False=ComponentMask(None=@n29) {b=true, desc=RockBaseColor, g=true, r=true}) {parameterName=UseCheapProjection}
   B=LinearInterpolate(
    A=VectorParameter() {group=Rock, parameterName=RockTint_1, defaultValue=<rgba 0.86 0.86 0.86 1>}.RGB
    B=VectorParameter() {group=Rock, parameterName=RockTint_2, defaultValue=<rgba 0.605 0.5144 0.3794 1>}.RGB
    Alpha=Clamp(None=@n34)
   )
  )
  n36 = MaterialFunctionCall(
   TextureObject=TextureObject() {texture=&WhiteRockCustom_D}
   TextureSize=ScalarParameter() {defaultValue=768, parameterName=TexSize_1}
   Export Float 4=StaticBool() {value=true}
  ) {materialFunction=&WorldAlignedTexture}
  n37 = ComponentMask(None=@n36.XYZ Texture) {b=true, g=true, r=true}
  n38 = ComponentMask(None=StaticSwitchParameter(True=@n31, False=@n29) {parameterName=UseCheapProjection}) {a=true, desc=RockHeight}
  n39 = Multiply(A=OneMinus(None=@n38), B=ScalarParameter() {defaultValue=2.75, parameterName=GravelRock_Transblend})
  n40 = VertexNormalWS()
  n41 = Power(Base=Multiply(A=ComponentMask(None=@n40) {b=true}, B=ScalarParameter() {defaultValue=1, parameterName=Gravel_Rock_Amount}) {constB=20})
  n42 = ComponentMask(None=@n36.XYZ Texture) {a=true}
  n43 = Power(Base=@n42) {constExponent=4}
  n44 = TextureSample(UVs=LandscapeLayerCoords() {mappingScale=6}) {texture=/Game/Packages/DS_Fortnight/SM/Textures/T_FORT_ForestFloor01}
  MaterialExpressionMultiply_35 = Multiply(A=@n37, B=VectorParameter() {parameterName=Mud_Tint, defaultValue=<rgba 0.575 0.575 0.575 0>}.RGB)
  n45 = Multiply(A=LandscapeLayerCoords() {mappingScale=8}, B=ScalarParameter() {defaultValue=2, parameterName=Grass_Tile})
  n46 = TextureSample(UVs=@n45) {texture=/Game/Packages/DS_Fortnite_Terrain_NoLOD/Textures/T_Grasslands_D}
  n47 = Subtract(A=ScalarParameter() {defaultValue=1.85, parameterName=GrassMudEdge_Transition}, B=@n46.A)
  n48 = MaterialFunctionCall(Float3=VertexNormalWS()) {materialFunction=&BreakOutFloat3Components}
  n49 = ScalarParameter() {defaultValue=-3.5, parameterName=GrassMudEdge_Amount}
  n50 = Constant() {r=1.1}
  n51 = Multiply(A=Power(Base=@n46.A) {constExponent=0.5}, B=ScalarParameter() {defaultValue=0.05, parameterName=Transition_Phase}) {constB=1.1}
  n52 = Multiply(
   A=MaterialFunctionCall(Float3=VertexNormalWS()) {materialFunction=&BreakOutFloat3Components}.B
   B=ScalarParameter() {defaultValue=2.1, parameterName=RockMultiply}
  )
  n53 = ScalarParameter() {defaultValue=5, parameterName=Contrast}
  MaterialExpressionMaterialFunctionCall_4 = MaterialFunctionCall(
   A=MaterialFunctionCall(
    A=ComponentMask(None=@n35) {b=true, desc=RockBaseColor, g=true, r=true}
    B=@n37
    Transition Phase=Multiply(A=@n39, B=@n41)
    Height Texture=Multiply(A=@n38, B=@n43)
    Contrast=ScalarParameter() {defaultValue=2, parameterName=GravelRock_Contrast}
   ) {materialFunction=&HeightLerp}.Results
   B=MaterialFunctionCall(
    A=Multiply(A=@n46.RGB, B=@n50) {constB=1.1}
    B=LinearInterpolate(A=@n44.RGB, B=@MaterialExpressionMultiply_35, Alpha=@n42)
    Transition Phase=Divide(A=@n47, B=@n48.B)
    Height Texture=Multiply(A=@n48.B, B=@n49)
    Contrast=ScalarParameter() {defaultValue=1, parameterName=GrassMudEdge_Contrast}
   ) {materialFunction=&HeightLerp}.Results
   Transition Phase=@n51
   Height Texture=@n52
   Contrast=@n53
  ) {materialFunction=&HeightLerp}
  n54 = OneMinus(None=@MaterialExpressionMaterialFunctionCall_4.Alpha)
  n55 = LandscapeLayerSample() {parameterName=Road}
  MaterialExpressionSubtract_2 = Subtract(A=Subtract(A=LandscapeLayerSample() {parameterName=Grass, previewWeight=1}, B=@n54), B=@n55)
  MaterialExpressionSubtract_4 = Subtract(A=Subtract(A=LandscapeLayerSample() {parameterName=GrassNoFoliage, previewWeight=1}, B=@n54), B=@n55)
  MaterialExpressionSubtract_6 = Subtract(A=LandscapeLayerSample() {parameterName=Mud, previewWeight=1}, B=@n55)
  MaterialExpressionSubtract_7 = Subtract(A=LandscapeLayerSample() {parameterName=Forest, previewWeight=1}, B=@n55)
  LandscapeGrassOutput!(
   Grass=@MaterialExpressionSubtract_2
   Grass_Standard=@MaterialExpressionSubtract_4
   Mud=@MaterialExpressionSubtract_6
   Forest=@MaterialExpressionSubtract_7
  ) {grassTypes [
    Grass {
     grassType=&/Athena_Terrain_Grass
     input {
      expression=&self:MaterialExpressionSubtract_2
      inputName=None
      mask=0
      maskA=0
      maskB=0
      maskG=0
      maskR=0
      outputIndex=0
     }
    }
    Grass_Standard {
     grassType=&/Athena_Terrain_Grass_Standard
     input {
      expression=&self:MaterialExpressionSubtract_4
      inputName=None
      mask=0
      maskA=0
      maskB=0
      maskG=0
      maskR=0
      outputIndex=0
     }
    }
    Mud {
     grassType=&/Athena_Terrain_Gravel
     input {
      expression=&self:MaterialExpressionSubtract_6
      inputName=None
      mask=0
      maskA=0
      maskB=0
      maskG=0
      maskR=0
      outputIndex=0
     }
    }
    Forest {
     grassType=&/Athena_Terrain_Forest
     input {
      expression=&self:MaterialExpressionSubtract_7
      inputName=None
      mask=0
      maskA=0
      maskB=0
      maskG=0
      maskR=0
      outputIndex=0
     }
    }
   ]}
  n56 = TextureSample(UVs=@n45) {samplerType=SAMPLERTYPE_Normal, texture=/Game/Packages/DS_Fortnite_Terrain_NoLOD/Textures/T_Grasslands_N}
  MakeMaterialAttributes!(
   BaseColor=@n46.RGB
   Metallic
   Specular=LinearInterpolate(
    A=ScalarParameter() {defaultValue=0.2, group=Grass, parameterName=SpecularGrass_Low}
    B=ScalarParameter() {defaultValue=0.2, group=Grass, parameterName=SpecularGrass_High}
    Alpha=@n46.R
   )
   Roughness=ScalarParameter() {group=Grass, parameterName=RoughnessGrass}
   Anisotropy
   EmissiveColor
   Opacity
   OpacityMask
   Normal=@n56.RGB
   Tangent
   WorldPositionOffset
   SubsurfaceColor
   ClearCoat
   ClearCoatRoughness
   AmbientOcclusion
   Refraction
   CustomizedUVs_0
   CustomizedUVs_1
   CustomizedUVs_2
   CustomizedUVs_3
   CustomizedUVs_4
   CustomizedUVs_5
   CustomizedUVs_6
   CustomizedUVs_7
   PixelDepthOffset
   ShadingModel
   Displacement
  )
  n57 = MaterialFunctionCall(
   TextureObject=TextureObject() {texture=/Game/Packages/DS_Fortnight/SM/Textures/T_Street_Road_D}
   TextureSize=ScalarParameter() {defaultValue=1024, parameterName=RoadSize}
   Export Float 4=StaticBool() {value=true}
  ) {materialFunction=&WorldAlignedTexture}
  n58 = ComponentMask(None=@n57.Z Texture) {a=true}
  Multiply(A=@n58)
  n59 = Constant() {r=0.8}
  n60 = LandscapeVisibilityMask()
  MaterialExpressionStaticSwitchParameter_0 = StaticSwitchParameter(
   True=MaterialFunctionCall!(Mask 2, UV Coordinates=LandscapeLayerCoords() {mappingScale=1525}, BaseTexture=@MaterialExpressionMaterialFunctionCall_4.Results) {materialFunction=&/MF_TerrainColorVariation}.Result
   False=@MaterialExpressionMaterialFunctionCall_4.Results
  ) {defaultValue=true, parameterName=UseTopoMasks}
  MaterialExpressionPower_0 = Power(Base=@n42)
  MaterialExpressionLinearInterpolate_9 = LinearInterpolate(A=@n44.RGB, B=@MaterialExpressionMultiply_35, Alpha=Multiply(A=@n42) {constB=0.85})
  MaterialExpressionMultiply_6 = Multiply(A=@n42, B=@n46.A)
  n61 = TextureSample(UVs=LandscapeLayerCoords() {mappingScale=4}) {texture=/Game/Packages/DS_Fortnight/Mud}
  n62 = LandscapeLayerSample() {parameterName=Road}
  n63 = Power(Base=@n62, Exp=ScalarParameter() {defaultValue=1, parameterName=Road_Pow}) {constExponent=3}
  n64 = Add(A=Multiply(A=@n63, B=@n58) {constB=0.8}, B=LandscapeLayerSample() {parameterName=Mud})
  n65 = ScalarParameter() {defaultValue=0.5, parameterName=TransitionMPRoad}
  MaterialExpressionMultiply_3 = Multiply(
   A=MaterialFunctionCall(
    A=Multiply(A=@n61.RGB) {constB=0.75}
    B=@n57.Z Texture
    Transition Phase=Multiply(A=@n64, B=@n65)
    Height Texture=Add(A=@n62, B=@n63)
    Contrast=ScalarParameter() {defaultValue=10, parameterName=RoadContrast}
   ) {materialFunction=&HeightLerp}.Results
   B=VectorParameter() {parameterName=Road_Tint, defaultValue=<rgba 1.13 1.13 1.13 0>}.RGB
  )
  n66 = LandscapeLayerBlend!(
   Layer Grass=@MaterialExpressionStaticSwitchParameter_0
   Layer GrassNoFoliage=@MaterialExpressionMaterialFunctionCall_4.Results
   Layer Mud=@MaterialExpressionMultiply_35
   Height Mud=@MaterialExpressionPower_0
   Layer Forest=@MaterialExpressionLinearInterpolate_9
   Height Forest=@MaterialExpressionMultiply_6
   Layer Road=@MaterialExpressionMultiply_3
  ) {layers [
    {
     blendType=LB_WeightBlend
     constHeightInput=0
     layerName=Grass
     previewWeight=0.5
     constLayerInput=<0 0 0>
     heightInput {expression=null, inputName=None, mask=0, maskA=0, maskB=0, maskG=0, maskR=0, outputIndex=0}
     layerInput {
      expression=&self:MaterialExpressionStaticSwitchParameter_0
      inputName=None
      mask=0
      maskA=0
      maskB=0
      maskG=0
      maskR=0
      outputIndex=0
     }
    }
    {
     blendType=LB_WeightBlend
     constHeightInput=0
     layerName=GrassNoFoliage
     previewWeight=0.5
     constLayerInput=<0 0 0>
     heightInput {expression=null, inputName=None, mask=0, maskA=0, maskB=0, maskG=0, maskR=0, outputIndex=0}
     layerInput {
      expression=&self:MaterialExpressionMaterialFunctionCall_4
      inputName=None
      mask=0
      maskA=0
      maskB=0
      maskG=0
      maskR=0
      outputIndex=0
     }
    }
    {
     blendType=LB_HeightBlend
     constHeightInput=0
     layerName=Mud
     previewWeight=0.5
     constLayerInput=<0 0 0>
     heightInput {
      expression=&self:MaterialExpressionPower_0
      inputName=None
      mask=0
      maskA=0
      maskB=0
      maskG=0
      maskR=0
      outputIndex=0
     }
     layerInput {
      expression=&self:MaterialExpressionMultiply_35
      inputName=None
      mask=0
      maskA=0
      maskB=0
      maskG=0
      maskR=0
      outputIndex=0
     }
    }
    {
     blendType=LB_HeightBlend
     constHeightInput=0
     layerName=Forest
     previewWeight=0.5
     constLayerInput=<0 0 0>
     heightInput {
      expression=&self:MaterialExpressionMultiply_6
      inputName=None
      mask=0
      maskA=0
      maskB=0
      maskG=0
      maskR=0
      outputIndex=0
     }
     layerInput {
      expression=&self:MaterialExpressionLinearInterpolate_9
      inputName=None
      mask=0
      maskA=0
      maskB=0
      maskG=0
      maskR=0
      outputIndex=0
     }
    }
    {
     blendType=LB_AlphaBlend
     constHeightInput=1
     layerName=Road
     previewWeight=0.5
     constLayerInput=<0 0 0>
     heightInput {expression=null, inputName=None, mask=1, maskA=1, maskB=0, maskG=0, maskR=0, outputIndex=4}
     layerInput {
      expression=&self:MaterialExpressionMultiply_3
      inputName=None
      mask=0
      maskA=0
      maskB=0
      maskG=0
      maskR=0
      outputIndex=0
     }
    }
   ]}
  MaterialExpressionLinearInterpolate_4 = LinearInterpolate(
   B=ScalarParameter() {defaultValue=0.3, group=Grass, parameterName=GrassSpecAmount}
   Alpha=Power(Base=ComponentMask(None=@n46.A) {a=true}, Exp=ScalarParameter() {defaultValue=2, group=Grass, parameterName=GrassSpecPower})
  )
  n67 = LandscapeLayerBlend!(Layer Grass=@MaterialExpressionLinearInterpolate_4) {layers [{
     blendType=LB_WeightBlend
     constHeightInput=0
     layerName=Grass
     previewWeight=0
     constLayerInput=<0 0 0>
     heightInput {expression=null, inputName=None, mask=0, maskA=0, maskB=0, maskG=0, maskR=0, outputIndex=0}
     layerInput {
      expression=&self:MaterialExpressionLinearInterpolate_4
      inputName=None
      mask=0
      maskA=0
      maskB=0
      maskG=0
      maskR=0
      outputIndex=0
     }
    }]}
  n68 = TextureSample(UVs=@n2) {
   samplerSource=SSM_Wrap_WorldGroupSettings
   samplerType=SAMPLERTYPE_Normal
   texture=/Game/Packages/DS_Fortnite_Terrain_NoLOD/Textures/T_Cave_Exterior_Wall_N
  }
  n69 = TextureSample(UVs=@n4) {
   samplerSource=SSM_Wrap_WorldGroupSettings
   samplerType=SAMPLERTYPE_Normal
   texture=&T_Cave_Exterior_Wall_N
  }
  n70 = TextureSample(UVs=@n10) {
   samplerSource=SSM_Wrap_WorldGroupSettings
   samplerType=SAMPLERTYPE_Normal
   texture=&T_Cave_Exterior_Wall_N
  }
  n71 = TextureSample(UVs=@n12) {
   samplerSource=SSM_Wrap_WorldGroupSettings
   samplerType=SAMPLERTYPE_Normal
   texture=&T_Cave_Exterior_Wall_N
  }
  n72 = LinearInterpolate(A=LinearInterpolate(A=@n68.RGB, B=@n69.RGB, Alpha=@n7.Result), B=LinearInterpolate(A=@n70.RGB, B=@n71.RGB, Alpha=@n7.Result), Alpha=@n18)
  n73 = LinearInterpolate(
   A=MaterialFunctionCall!(
    TextureObject=TextureObject() {samplerType=SAMPLERTYPE_Normal, texture=&T_Cave_Exterior_Wall_N}
    TextureSize=@n19
    WorldSpace
    Normal
    ProjectionTransitionContrast=ScalarParameter() {defaultValue=-10, parameterName=RockProjContrast_Normal}
    WorldPosition
    Use High Quality Normals
   ) {materialFunction=/Game/Athena/Environments/Materials/WorldAlignedNormalContrast}.XYZFlatTop
   B=MaterialFunctionCall(
    TextureObject=TextureObject() {samplerType=SAMPLERTYPE_Normal, texture=&T_Cave_Exterior_Wall_N}
    TextureSize=@n24
    ProjectionTransitionContrast=ScalarParameter() {defaultValue=-10, parameterName=RockProjContrast_Normal}
   ) {materialFunction=&WorldAlignedNormalContrast}.XYZFlatTop
   Alpha=@n18
  )
  MaterialExpressionMaterialFunctionCall_7 = MaterialFunctionCall(
   A=ComponentMask(None=StaticSwitchParameter(True=@n72, False=@n73) {parameterName=UseCheapProjection}) {b=true, desc=Rock Normal, g=true, r=true}
   B=@n56.RGB
   Transition Phase=@n51
   Height Texture=@n52
   Contrast=@n53
  ) {materialFunction=&HeightLerp}
  MaterialExpressionConstant3Vector_0 = Constant3Vector() {constant=<rgba 0.5 0.5 1 0>}
  n74 = LandscapeLayerBlend!(
   Layer Grass=@MaterialExpressionMaterialFunctionCall_7.Results
   Layer GrassNoFoliage=@MaterialExpressionMaterialFunctionCall_7.Results
   Layer Mud=@MaterialExpressionMaterialFunctionCall_7.Results
   Height Mud=@MaterialExpressionPower_0
   Layer Forest=@MaterialExpressionMaterialFunctionCall_7.Results
   Height Forest=@MaterialExpressionMultiply_6
   Layer Road=@MaterialExpressionConstant3Vector_0
  ) {layers [
    {
     blendType=LB_WeightBlend
     constHeightInput=0
     layerName=Grass
     previewWeight=0.5
     constLayerInput=<0 0 0>
     heightInput {expression=null, inputName=None, mask=0, maskA=0, maskB=0, maskG=0, maskR=0, outputIndex=0}
     layerInput {
      expression=&self:MaterialExpressionMaterialFunctionCall_7
      inputName=None
      mask=0
      maskA=0
      maskB=0
      maskG=0
      maskR=0
      outputIndex=0
     }
    }
    {
     blendType=LB_WeightBlend
     constHeightInput=0
     layerName=GrassNoFoliage
     previewWeight=0.5
     constLayerInput=<0 0 0>
     heightInput {expression=null, inputName=None, mask=0, maskA=0, maskB=0, maskG=0, maskR=0, outputIndex=0}
     layerInput {
      expression=&self:MaterialExpressionMaterialFunctionCall_7
      inputName=None
      mask=0
      maskA=0
      maskB=0
      maskG=0
      maskR=0
      outputIndex=0
     }
    }
    {
     blendType=LB_HeightBlend
     constHeightInput=0
     layerName=Mud
     previewWeight=0.5
     constLayerInput=<0 0 0>
     heightInput {
      expression=&self:MaterialExpressionPower_0
      inputName=None
      mask=0
      maskA=0
      maskB=0
      maskG=0
      maskR=0
      outputIndex=0
     }
     layerInput {
      expression=&self:MaterialExpressionMaterialFunctionCall_7
      inputName=None
      mask=0
      maskA=0
      maskB=0
      maskG=0
      maskR=0
      outputIndex=0
     }
    }
    {
     blendType=LB_HeightBlend
     constHeightInput=0
     layerName=Forest
     previewWeight=0.5
     constLayerInput=<0 0 0>
     heightInput {
      expression=&self:MaterialExpressionMultiply_6
      inputName=None
      mask=0
      maskA=0
      maskB=0
      maskG=0
      maskR=0
      outputIndex=0
     }
     layerInput {
      expression=&self:MaterialExpressionMaterialFunctionCall_7
      inputName=None
      mask=0
      maskA=0
      maskB=0
      maskG=0
      maskR=0
      outputIndex=0
     }
    }
    {
     blendType=LB_AlphaBlend
     constHeightInput=1
     layerName=Road
     previewWeight=0.5
     constLayerInput=<0 0 0>
     heightInput {expression=null, inputName=None, mask=1, maskA=1, maskB=0, maskG=0, maskR=0, outputIndex=4}
     layerInput {
      expression=&self:MaterialExpressionConstant3Vector_0
      inputName=None
      mask=0
      maskA=0
      maskB=0
      maskG=0
      maskR=0
      outputIndex=0
     }
    }
   ]}
  RuntimeVirtualTextureOutput!(BaseColor=@n66, Specular=@n67, Roughness=@n59, Normal=@n74, WorldHeight, Opacity, Mask=@n60, Displacement, Mask4)
  TextureSample(UVs=@n2) {samplerSource=SSM_Wrap_WorldGroupSettings, texture=&WhiteRockCustom_D}
  TextureSample(UVs=@n4) {samplerSource=SSM_Wrap_WorldGroupSettings, texture=&WhiteRockCustom_D}
  n75 = RuntimeVirtualTextureSampleParameter() {materialType=BaseColor_Normal_Specular_Mask_YCoCg, parameterName=RVT, virtualTexture=&/RVT_Figment}
  n76 = SetMaterialAttributes!(
   MaterialAttributes
   Base Color=@n75.BaseColor
   Specular=@n75.BaseColor
   Roughness=@n75.BaseColor
   Opacity Mask=@n75.BaseColor
   Normal=@n75.BaseColor
  ) {attributeSetTypes [
    #1
    #2
    #3
    #4
    #5
   ]}
  n77 = SetMaterialAttributes(Base Color=@n66, Specular=@n67, Roughness=@n59, Opacity Mask=@n60, Normal=@n74) {attributeSetTypes [
    #1
    #2
    #3
    #4
    #5
   ]}
  VirtualTextureFeatureSwitch!(No=@n77, Yes=@n76)
 ]
}
